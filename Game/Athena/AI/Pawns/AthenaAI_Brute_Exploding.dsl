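; Athena AI pawn for an exploding Brute variant that shows a cosmetic explosion when it dies.
; BeginPlay binds ExplodeOnDeath to the OnDeathEffects dispatcher.
; That handler runs only on the non-authority (Remote) branch, activating the grenade launcher explosion component and playing the MASH Brute explode sound at the actor's location.
; It applies no damage here; the overlap and tick events are disabled stubs.
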(event EventBeginPlay
  (|CallParentFunction)
  (EventDispatchers|BindEventtoOnDeathEffects (delegate Custom|ExplodeOnDeath)))

(event Custom|ExplodeOnDeath
  (params (DeadActor :type Actor%20Object%20Reference) (Damage :type Float%20%28single-precision%29) (InstigatedBy :type Fort%20Pawn%20Object%20Reference) (HitLocation :type Vector) (DamageCauser :type Actor%20Object%20Reference))
  (Networking|SwitchHasAuthority
    (:Remote
      (Components|Activation|SetActive (Variables|Default|GetP_GrenadeLauncher_Explosion_Athena) true false)
      (Audio|PlaySoundatLocation "/Game/Athena/Sounds/LTMs/MASH/MASH_Brute_Explode_Cue.MASH_Brute_Explode_Cue" (Transformation|GetActorLocation)))))

(event Collision|EventActorBeginOverlap
  (params (OtherActor :type Actor%20Object%20Reference))
  (|CallParentFunction OtherActor) ; This node is disabled and will not be called.\nDrag off pins to build functionality.
)

(event EventTick
  (params (DeltaSeconds :type Float%20%28single-precision%29))
  (|CallParentFunction DeltaSeconds) ; This node is disabled and will not be called.\nDrag off pins to build functionality.
)

(fn ConstructionScript
  (|CallParentFunction))
